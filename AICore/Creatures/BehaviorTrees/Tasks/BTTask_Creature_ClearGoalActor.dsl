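; Behavior tree task for creature AI that clears the AI's current goal and assignment.
; On execute, it validates the owner actor, casts its AI controller to FortAIController and calls ClearCurrentGoalandAssignment on that controller's AI goal component.
; It finishes with success after clearing, or fails when the owner is invalid or the cast fails.
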(event EventReceiveExecute
  (params (OwnerActor :type Actor%20Object%20Reference))
  (Utilities|IsValid OwnerActor
    (:Is%20Valid
      (bind _asfort_aicontroller (Utilities|Casting|CastToFortAIController (AI|GetAIController OwnerActor))
        (:then
          (AI|ClearCurrentGoalandAssignment (Variables|/Script/FortniteAI.FortAIController|GetAIGoalComponent _asfort_aicontroller) true)
          (AI|BehaviorTree|FinishExecute true))
        (:CastFailed
          _)))
    (:Is%20Not%20Valid
      _))
  (AI|BehaviorTree|FinishExecute false) ; Fail Task
)
